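# Material function that spins the motor part of the WE Zipline gameplay prop through World Position Offset.
# One-second looping time, scaled by a rotation speed parameter and masked by the red channel of the motor mask texture, gives the angle.
# Vertices rotate about the local X axis around an offset pivot parameter transformed to world space.
# FixRotateAboutAxisNormals corrects the normals, and both results override the incoming material attributes.
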MaterialFunction /WildEstateTech/Environment/Props/Gameplay/Gameplay_WEZipline_A/Materials/MF_WEZipline_MotorWPO
graph {nodes [
  n1 = Transform(None=Constant3Vector() {constant=<rgba 1 0 0 0>}) {transformSourceType=TRANSFORMSOURCE_Local}
  n2 = Multiply(
   A=Multiply(A=Time() {bOverride_Period=true, period=1}, B=ScalarParameter() {defaultValue=-1, parameterName=ZiplineMotor.RotationSpeed})
   B=TextureSample() {texture=/WildEstateTech/Environment/Props/Gameplay/Gameplay_WEZipline_A/Textures/T_Gameplay_WEZiplineMotor_Mask}.R
  )
  n3 = TransformPosition(None=VectorParameter() {parameterName=ZiplineMotor.Offset, defaultValue=<rgba 0 0 -12.2 1>}.RGB) {transformType=TRANSFORMPOSSOURCE_World}
  n4 = WorldPosition()
  FunctionOutput(None=SetMaterialAttributes!(
    MaterialAttributes=FunctionInput() {inputName=Material, inputType=FunctionInput_MaterialAttributes}
    Normal=MaterialFunctionCall(Rotation Axis=@n1, Rotation Angle=@n2) {materialFunction=&FixRotateAboutAxisNormals}.New World Space Vertex Normal
    WorldPositionOffset=RotateAboutAxis(NormalizedRotationAxis=@n1, RotationAngle=@n2, PivotPoint=@n3, Position=@n4.XYZ)
   ) {attributeSetTypes [#1, #2]})
 ]}
